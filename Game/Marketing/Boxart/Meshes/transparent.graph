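# Translucent, Default Lit material named transparent, stored among the marketing box art mesh assets.
# Its graph holds a single scalar Constant of low value, with no textures, parameters or math.
# No link from that constant to a material output is shown.
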Material /Game/Marketing/Boxart/Meshes/transparent
graph {properties {blendMode=BLEND_Translucent, pixelDepthOffsetMode=PDOM_Legacy, usedShadingModels=Default Lit}, nodes [Constant() {r=0.04}]}
